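# Sound cue played when a Takedown Medallion is dropped.
# A single wave player for the default drop sound feeds a modulator at the root.
# It uses weapon drop attenuation and the loot collect sound class, at slightly reduced volume.
# A concurrency override limits it to one playing instance at a time.
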SoundCue /TakedownMedallion/Audio/Medallions/SC_TakedownMedallions_Drop_Cue
root=Modulator_0
nodes [
 {id=Modulator_0, children [WavePlayer_0]}
 {id=WavePlayer_0, properties {soundWaveAssetPtr=/TakedownMedallion/Audio/Medallions/S28/Default_TakedownMedallions_Drop}}
]
properties {
 attenuationSettings=&Fort_Weapon_Drop_Att
 bOverrideConcurrency=true
 maxDistance=3000
 soundClassObject=&Fort_Loot_Collect
 volumeMultiplier=0.8
 concurrencyOverrides {maxCount=1}
}
