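# Sound cue for the Takedown Medallion pickup VFX flourish, with an empty node graph and no root, so it references no audio.
# It carries playback settings only: shared weapon-pickup attenuation plus overrides, the ambient loot loop sound class and a debug flag.
# A concurrency override caps simultaneous instances at two, with distance-based volume scaling and a short release on stolen voices.
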SoundCue /TakedownMedallion/Audio/Medallions/SC_TakedownMedallions_Pickup_VFX_Flourish
root=null
nodes []
properties {
 attenuationSettings=&Fort_Weapon_Pickup_Att
 bDebug=true
 bOverrideConcurrency=true
 maxDistance=3500
 soundClassObject=/Game/Sounds/FortSoundClassesAndModes/Fort_Loot_Ambient_Loop
 volumeMultiplier=1
 attenuationOverrides {falloffDistance=512, nonSpatializedRadiusStart=20, stereoSpread=420}
 concurrencyOverrides {maxCount=2, voiceStealReleaseTime=0.3, volumeScale=0.4, volumeScaleAttackTime=0.1, volumeScaleMode=Distance}
}
